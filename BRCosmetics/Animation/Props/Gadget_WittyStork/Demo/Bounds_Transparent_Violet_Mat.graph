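# Translucent, default-lit violet material in the WittyStork gadget's Demo folder, apparently for showing see-through bounds, as its name suggests.
# It defines a scalar opacity parameter defaulting to half and a vector parameter holding a dark violet color, with no textures or math nodes.
# Usage flags enable it on skeletal meshes and with mesh deformers.
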Material /BRCosmetics/Animation/Props/Gadget_WittyStork/Demo/Bounds_Transparent_Violet_Mat
graph {
 properties {
  bUsedWithMeshDeformer=true
  bUsedWithSkeletalMesh=true
  blendMode=BLEND_Translucent
  pixelDepthOffsetMode=PDOM_Legacy
  usedShadingModels=Default Lit
 }
 nodes [
  ScalarParameter() {defaultValue=0.5, parameterName=opacity}
  VectorParameter() {defaultValue=<rgba 0.1095 0.0452 0.1384 1>}
 ]
}
